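; Parent gameplay ability for Large Spire boss abilities, providing shared activation, hit-react and cooldown handling.
; Activation caches the spire avatar, commits, applies an active effect, binds a damage-threshold handler that only logs, rolls random cooldown and prevent durations, and sets the spire's hit-react montage.
; Ending removes the active effect and applies cooldown and prevent effects with those durations via SetByCaller.
; A helper exposes a scalable-float weakpoint toggle flag.
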(event Ability|EventActivateAbility
  (Development|PrintString "@@@ Event ActivateAbility [GA_Large_Spire_Ability_Parent]: CALLED!" false (Variables|Debug|GetbDebugAbility))
  (CallFunction|InitSpireAbilityVariables)
  (CallFunction|SetupSpireAbility)
  (CallFunction|SetupAbilityHitReact))

(event Ability|EventOnEndAbility
  (params (bWasCancelled :type Boolean))
  (Development|PrintString "@@@ Event OnEndAbility [GA_Large_Spire_Ability_Parent]: CALLED!" false (Variables|Debug|GetbDebugAbility))
  (Ability|RemoveGameplayEffectFromOwnerWithHandle (Variables|Internal|GetAbilityActiveHandle))
  (bind _returnvalue (Ability|GameplayEffect|AssignTagSetbyCallerMagnitude (Ability|MakeOutgoingGameplayEffectSpec (Variables|Tunables|GetGE_AbilityCooldown)) "(TagName=\"SetByCaller.Duration\")" (Variables|Internal|GetAbilityCooldownTime)))
  (bind _returnvalue_2 (Ability|ApplyGameplayEffectSpecToOwner _returnvalue))
  (Variables|Internal|SetAbilityCooldownHandle _returnvalue_2)
  (bind _returnvalue_3 (Ability|GameplayEffect|AssignTagSetbyCallerMagnitude (Ability|MakeOutgoingGameplayEffectSpec (Variables|Internal|GetGE_AbilityPrevent)) "(TagName=\"SetByCaller.Duration\")" (Variables|Internal|GetAbilityPreventTime)))
  (bind _returnvalue_4 (Ability|ApplyGameplayEffectSpecToOwner _returnvalue_3))
  (Variables|Internal|SetAbilityPreventHandle _returnvalue_4))

(event Custom|SpireDamageThresholdReached
  (params (DamageTags :type Gameplay%20Tag%20Container%20Structure) (HitResult :type Hit%20Result%20Structure) (Instigator :type Controller%20Object%20Reference) (DamageCauser :type Actor%20Object%20Reference))
  (Development|PrintString "@@@ SpireDamageThresholdReached [GA_Large_Spire_Ability_Parent]: CALLED!" false (Variables|Debug|GetbDebugAbility)))

(fn GetAbilityCooldownTime
  (return (Math|Random|RandomFloatinRange (ScalableFloat|GetValueatLevel (Variables|Tunables|GetSF_CooldownMin)) (ScalableFloat|GetValueatLevel (Variables|Tunables|GetSF_CooldownMax)))))

(fn GetAbilityPreventTime
  (return (Math|Random|RandomFloatinRange (ScalableFloat|GetValueatLevel (Variables|Internal|GetSF_PreventMin)) (ScalableFloat|GetValueatLevel (Variables|Internal|GetSF_PreventMax)))))

(fn InitSpireAbilityVariables
  (bind _asbga_large_spire (Utilities|Casting|CastToBGA_Large_Spire (Ability|GetAvatarActorfromActorInfo))
    (:then
      (Variables|Internal|SetSpireActor _asbga_large_spire))
    (:CastFailed
      (Ability|EndAbility))))

(fn SetupAbilityHitReact
  (Variables|/SpireBoss/GameplayActors/BGA_Large_Spire.BGA_Large_Spire_C|SetCurrentSpireHitReact (Variables|Tunables|GetAbilityHitReactMontage) (Variables|Internal|GetSpireActor)))

(fn SetupSpireAbility
  (bind _bdebugability (Variables|Debug|GetbDebugAbility))
  (Development|PrintString "@@@@@@ SetupSpireAbility [GA_Large_Spire_Ability_Parent]: CALLED!" false (Variables|Debug|GetbDebugAbility))
  (bind _returnvalue (Ability|CommitAbility))
  (if _returnvalue
    (bind _returnvalue_2 (Ability|ApplyGameplayEffectToOwner (Variables|Tunables|GetGE_AbilityActive)))
    (Variables|Internal|SetAbilityActiveHandle _returnvalue_2)
    (Default|BindEventtoOnSpireDamageThresholdReached (Variables|Internal|GetSpireActor) (EventDispatchers|CreateEvent :!event SpireDamageThresholdReached))
    (Utilities|FlowControl|Sequence
      (:then_0
        (bind _output_get (Variables|Internal|SetAbilityCooldownTime (CallFunction|GetAbilityCooldownTime)))
        (Development|PrintString (Utilities|String|Append "@@@@@@@@@ SetupSpireAbility [GA_Large_Spire_Ability_Parent]: Ability Cooldown = " (Utilities|String|ToString(Float) _output_get) " seconds!") false _bdebugability))
      (:then_1
        (bind _output_get_2 (Variables|Internal|SetAbilityPreventTime (CallFunction|GetAbilityPreventTime)))
        (Development|PrintString (Utilities|String|Append "@@@@@@@@@@@@ SetupSpireAbility [GA_Large_Spire_Ability_Parent]: Ability Prevent Time = " (Utilities|String|ToString(Float) _output_get_2) " seconds!") false _bdebugability)))
    (else
      (Ability|EndAbility))))

(fn ShouldToggleWeakpoints
  (return (ScalableFloat|GetValueasBool (Variables|Internal|GetSF_UseAbilityWeakpointToggles))))
