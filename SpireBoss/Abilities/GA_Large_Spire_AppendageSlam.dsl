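; Boss ability where the Large Spire slams the appendage nearest a valid player target, within range and not downed or cloaked.
; It plays that appendage's montage with an anticipation cue; the fire notify sets a navmesh-projected impact at the bone.
; The slam damages and launches pawns and vehicles, spawns a ground wave ignoring those hit, and exposes a weakpoint until recovery.
; Optional EQS-placed spikes spawn on a chained timer under a count cap.
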(event Ability|EventActivateAbility
  (bind _currentclosestappendagekind (Variables|Internal|GetCurrentClosestAppendageKind))
  (bind _spireactor (Variables|Internal|GetSpireActor))
  (bind (_targetstructdata_spireappendagekind_2_c321b3314c9291adb879629b5339498e _targetstructdata_bonename_5_38cd39514d738e05454978b3e574f8a0 _targetstructdata_slammontage_8_9f3a583242b3080b17ff53800eba48b2 _targetstructdata_weakpointkind_14_3e0154dc496b6f9415a82583af261852 _targetstructdata_hitreactmontage_17_937f2f8e42a39767ddabfa9e74c788a2) (Utilities|Struct|BreakSpireAppendageStruct (Class|BGALargeSpire|GetAppendageDatabyKind _spireactor _currentclosestappendagekind)))
  (bind (_bfoundtarget _slamtarget) (CallFunction|GetValidSlamTarget))
  (bind _spireactor_2 (Variables|Internal|GetSpireActor))
  (|CallParentFunction)
  (if _bfoundtarget
    (bind _closestappendagekind (Class|BGALargeSpire|GetClosestAppendageKind (Variables|Internal|GetSpireActor) (Pawn|GetNavAgentLocation _slamtarget)))
    (bind _output_get (Variables|Internal|SetCurrentClosestAppendageKind _closestappendagekind))
    (Variables|/SpireBoss/GameplayActors/BGA_Large_Spire.BGA_Large_Spire_C|SetCurrentClosestAppendageKind _output_get _spireactor_2)
    (Variables|/SpireBoss/GameplayActors/BGA_Large_Spire.BGA_Large_Spire_C|SetCurrentSpireHitReact _targetstructdata_hitreactmontage_17_937f2f8e42a39767ddabfa9e74c788a2 _spireactor_2)
    (bind (_asynctaskproxy _notifytag) (Ability|Tasks|PlayMontageAndWaitNotify "SpireBoss_AppendageSlam" _targetstructdata_slammontage_8_9f3a583242b3080b17ff53800eba48b2 1.0 "Intro")
      (:then
        (Variables|Internal|SetAppendageSlamMontageTask _asynctaskproxy)
        (GameplayCue|AddGameplayCueOnActor(Looping) (Variables|Internal|GetSpireActor) (Variables|Internal|GetGCTag_Anticipation) (Ability|GameplayCue|MakeGameplayCueParameters :NormalizedMagnitude 0.0 :RawMagnitude 0.0 :Location (Pawn|GetNavAgentLocation (Variables|/SpireBoss/GameplayActors/BGA_Large_Spire.BGA_Large_Spire_C|GetCurrentTargetPlayerPawn (Variables|Internal|GetSpireActor))) :Normal "0,0,0" :GameplayEffectLevel 1 :AbilityLevel 1 :bReplicateLocationWhenUsingMinimalRepProxy false)))
      (:OnCompleted
        _)
      (:OnInterrupted
        _)
      (:OnCancelled
        _)
      (:OnNotifyBegin
        (if (== _notifytag (Variables|Internal|GetMontageTag_Fire))
          (CallFunction|TriggerSlam)
          (GameplayCue|RemoveGameplayCueOnActor(Looping) (Variables|Internal|GetSpireActor) (Variables|Internal|GetGCTag_Anticipation) (Ability|GameplayCue|MakeGameplayCueParameters 0.0)))))
    (Ability|EndAbility)
    (else
      (Ability|EndAbility))))

(event Custom|DealInstantDamage
  (bind (_targetstructdata_spireappendagekind_2_c321b3314c9291adb879629b5339498e _targetstructdata_bonename_5_38cd39514d738e05454978b3e574f8a0) (Utilities|Struct|BreakSpireAppendageStruct (Class|BGALargeSpire|GetAppendageDatabyKind (Variables|Internal|GetSpireActor) (Variables|Internal|GetCurrentClosestAppendageKind))))
  (bind _returnvalue_location (Math|Transform|BreakTransform (Transformation|GetBoneTransform (Variables|/SpireBoss/GameplayActors/BGA_Large_Spire.BGA_Large_Spire_C|GetSkeletalMeshComponentBudgeted (Variables|Internal|GetSpireActor)) _targetstructdata_bonename_5_38cd39514d738e05454978b3e574f8a0 "RTS_World")))
  (bind (_asynctaskproxy _targetdata) (Ability|Tasks|DoTargetSelection "InstantAOEImpact" "(TagName=\"Targeting.ApplicationTag.0\")" true)
    (:Targeted
      (bind _output_get (Variables|Internal|SetInstantAOETargets (Ability|TargetData|GetAllActorsfromTargetData _targetdata)))
      (for _array_element _output_get
        (bind _asfort_pawn (Utilities|Casting|CastToFortPawn _array_element)
          (:then
            (Utilities|IsinVehicle _asfort_pawn
              (:false
                (if (CallFunction|ShouldInstantDamageAffectPawn _asfort_pawn)
                  (Variables|Internal|SetTargetIterLoc (Pawn|GetNavAgentLocation _asfort_pawn))
                  (Character|LaunchCharacter _asfort_pawn (+ (Math|Vector|vector+vector (* (Math|Vector|Normalize (- (Variables|Internal|GetTargetIterLoc) (.location (Variables|Internal|GetImpactTransform)))) (ScalableFloat|GetValueatLevel (Variables|Internal|GetSF_InstantAOE_PlayerLaunchVelocity))) (Math|Vector|MakeVector 0.0 0.0 (ScalableFloat|GetValueatLevel (Variables|Internal|GetSF_InstantAOE_PlayerLaunchVelocityAddedZ)))) (Math|Vector|Normalize (- (Variables|Internal|GetTargetIterLoc) (.location (Variables|Internal|GetImpactTransform))))) true true)
                  (else
                    (Utilities|Array|RemoveItem (Variables|Internal|GetInstantAOETargets) _array_element))))))
          (:CastFailed
            (bind _asfort_athena_vehicle (Utilities|Casting|CastToFortAthenaVehicle _array_element)
              (:then
                (Audio|BroadcastAbilityImpactatLocation 0 _asfort_athena_vehicle (+ (Math|Vector|vector+vector (* (Math|Vector|Normalize (- (Transformation|GetActorLocation _asfort_athena_vehicle) (AI|Navigation|ProjectPointtoNavigation _returnvalue_location 0 "/Game/Athena/AI/NavFilters/NavFilter_NPC_AnyTraversable.NavFilter_NPC_AnyTraversable_C"))) (ScalableFloat|GetValueatLevel (Variables|Internal|GetSF_InstantAOE_VehicleLaunchVelocity))) (Math|Vector|MakeVector 0.0 0.0 (ScalableFloat|GetValueatLevel (Variables|Internal|GetSF_InstantAOE_VehicleLaunchVelocityAddedZ)))) (Math|Vector|Normalize (- (Transformation|GetActorLocation _asfort_athena_vehicle) (AI|Navigation|ProjectPointtoNavigation _returnvalue_location 0 "/Game/Athena/AI/NavFilters/NavFilter_NPC_AnyTraversable.NavFilter_NPC_AnyTraversable_C")))) 777.0)
                (for _array_element_2 (Vehicle|GetAllPassengers _asfort_athena_vehicle)
                  (Utilities|IsValid _array_element_2
                    (:Is%20Valid
                      (Utilities|Array|AddUnique (Variables|Internal|GetInstantAOETargets) _array_element_2)))))
              (:CastFailed
                _))))
        (CallFunction|ApplySlamDamage (Ability|TargetData|AbilityTargetDatafromHitResult (Collision|MakeHitResult false false 0.0 0.0 (Transformation|GetActorLocation _array_element) (Transformation|GetActorLocation _array_element) "0,0,1" "0,0,1" 0 _array_element)))))))

(event Custom|TriggerSlam
  (bind (_targetstructdata_spireappendagekind_2_c321b3314c9291adb879629b5339498e _targetstructdata_bonename_5_38cd39514d738e05454978b3e574f8a0) (Utilities|Struct|BreakSpireAppendageStruct (Class|BGALargeSpire|GetAppendageDatabyKind (Variables|Internal|GetSpireActor) (Variables|Internal|GetCurrentClosestAppendageKind))))
  (bind (_returnvalue_location _returnvalue_rotation _returnvalue_scale) (Math|Transform|BreakTransform (Transformation|GetBoneTransform (Variables|/SpireBoss/GameplayActors/BGA_Large_Spire.BGA_Large_Spire_C|GetSkeletalMeshComponentBudgeted (Variables|Internal|GetSpireActor)) _targetstructdata_bonename_5_38cd39514d738e05454978b3e574f8a0 "RTS_World")))
  (bind (_targetstructdata_spireappendagekind_2_c321b3314c9291adb879629b5339498e_2 _targetstructdata_bonename_5_38cd39514d738e05454978b3e574f8a0_2 _targetstructdata_slammontage_8_9f3a583242b3080b17ff53800eba48b2 _targetstructdata_weakpointkind_14_3e0154dc496b6f9415a82583af261852) (Utilities|Struct|BreakSpireAppendageStruct (Class|BGALargeSpire|GetAppendageDatabyKind (Variables|Internal|GetSpireActor) (Variables|Internal|GetCurrentClosestAppendageKind))))
  (Development|PrintString "@@@@@@ TriggerSlam [GA_Large_Spire_AppendageSlam]: CALLED!" false (Variables|Debug|GetbDebugAbility))
  (Utilities|FlowControl|Sequence
    (:then_0
      (Variables|Internal|SetImpactTransform (Math|Transform|MakeTransform (AI|Navigation|ProjectPointtoNavigation _returnvalue_location 0 "/Game/Athena/AI/NavFilters/NavFilter_NPC_AnyTraversable.NavFilter_NPC_AnyTraversable_C" "0.000000,0.000000,1000.000000") _returnvalue_rotation _returnvalue_scale)))
    (:then_1
      (CallFunction|DealInstantDamage))
    (:then_2
      (if (ScalableFloat|GetValueasBool (Variables|SpireSpike|GetSF_ShouldSpawnSpikes))
        (bind _returnvalue (AI|EQS|RunEQSQuery "/SpireBoss/AI/EQS/EQS_Large_Spire_Spike_SpawnLocation.EQS_Large_Spire_Spike_SpawnLocation" self "AllMatching"))
        (AI|EQS|BindEventtoOnQueryFinished _returnvalue (delegate Custom|OnEnvQueryFinished))))
    (:then_3
      (bind _spawnedactor (CallFunction|SpawnGroundWaveActor))
      (bind _asbga_large_spire_ground_wave (Utilities|Casting|CastToBGA_Large_Spire_GroundWave _spawnedactor))
      (Variables|/SpireBoss/GameplayActors/BGA_Large_Spire_GroundWave.BGA_Large_Spire_GroundWave_C|SetWaveIgnoreActors (Variables|Internal|GetInstantAOETargets) _asbga_large_spire_ground_wave))
    (:then_4
      (if (Variables|Debug|GetbDebugAbility)
        (Rendering|Debug|DrawDebugSphere (.location (Variables|Internal|GetImpactTransform)) 5e+01 12 "(R=1.000000,G=0.027003,B=0.636845,A=1.000000)" 1e+01 2.0))
      (GameplayCue|ExecuteGameplayCueOnActor(Burst) (Variables|Internal|GetSpireActor) (Variables|Internal|GetGCTag_InstantAOE) (Ability|GameplayCue|MakeGameplayCueParameters 0.0))
      (if (CallFunction|ShouldToggleWeakpoints)
        (Class|BGALargeSpire|UpdateChildWeakpointActorbyKind (Variables|Internal|GetSpireActor) _targetstructdata_weakpointkind_14_3e0154dc496b6f9415a82583af261852 true)
        (Variables|Internal|SetbWeakpointsActivated true)))
    (:then_5
      (CallFunction|StartSlamRecovery))))

(event Custom|OnEnvQueryFinished
  (params (QueryInstance :type EQS%20Query%20Instance%20Object%20Reference) (QueryStatus :type EEnvQueryStatus%20Enum :default "Processing"))
  (bind _spirespikespawnlocation (Variables|SpireSpike|GetSpireSpikeSpawnLocation))
  (bind _resultlocations (AI|EQS|GetQueryResultsasLocations QueryInstance))
  (bind _array_element (Utilities|Array|ForEachLoopwithBreak _resultlocations)
    (:LoopBody
      (Utilities|Array|Add _spirespikespawnlocation _array_element)
      (if (>= (Utilities|Array|Length _spirespikespawnlocation) (Variables|SpireSpike|GetSpireSpikeToSpawn))
        _))
    (:Completed
      _))
  (bind _returnvalue (Utilities|Time|SetTimerbyEvent (EventDispatchers|CreateEvent :!event Custom|SpawnSpirePike) (ScalableFloat|GetValueatLevel (Variables|SpireSpike|GetDelayBetweenSpireSpikePawn)) false true))
  (Variables|SpireSpike|SetSpawnSpireSpikeTimer _returnvalue))

(event Custom|StartSlamRecovery
  ; Slam Recovery
  (Utilities|FlowControl|Delay (ScalableFloat|GetValueatLevelwithContextParam (Ability|GetAvatarActorfromActorInfo) (Variables|Internal|GetSF_SlamRecoveryTime)))
  (CallFunction|PostSlamRecovery))

(event Custom|PostSlamRecovery
  (bind (_targetstructdata_spireappendagekind_2_c321b3314c9291adb879629b5339498e _targetstructdata_bonename_5_38cd39514d738e05454978b3e574f8a0 _targetstructdata_slammontage_8_9f3a583242b3080b17ff53800eba48b2 _targetstructdata_weakpointkind_14_3e0154dc496b6f9415a82583af261852) (Utilities|Struct|BreakSpireAppendageStruct (Class|BGALargeSpire|GetAppendageDatabyKind (Variables|Internal|GetSpireActor) (Variables|Internal|GetCurrentClosestAppendageKind))))
  ; Slam Recovery
  (Ability|Animation|MontageJumptoSection "Outro")
  (if (CallFunction|ShouldToggleWeakpoints)
    (Class|BGALargeSpire|UpdateChildWeakpointActorbyKind (Variables|Internal|GetSpireActor) _targetstructdata_weakpointkind_14_3e0154dc496b6f9415a82583af261852 false)
    (Variables|Internal|SetbWeakpointsActivated false)))

(event Ability|EventOnEndAbility
  (params (bWasCancelled :type Boolean))
  (bind (_targetstructdata_spireappendagekind_2_c321b3314c9291adb879629b5339498e _targetstructdata_bonename_5_38cd39514d738e05454978b3e574f8a0 _targetstructdata_slammontage_8_9f3a583242b3080b17ff53800eba48b2 _targetstructdata_weakpointkind_14_3e0154dc496b6f9415a82583af261852) (Utilities|Struct|BreakSpireAppendageStruct (Class|BGALargeSpire|GetAppendageDatabyKind (Variables|Internal|GetSpireActor) (Variables|Internal|GetCurrentClosestAppendageKind))))
  (|CallParentFunction bWasCancelled)
  (Development|PrintString "@@@@@@@@@ Event OnEndAbility [GA_Large_Spire_AppendageSlam]: CALLED!" false (Variables|Debug|GetbDebugAbility))
  (Ability|RemoveGameplayEffectFromOwnerWithHandle (Variables|Internal|GetSlamActiveHandle))
  (if (and (Variables|Internal|GetbWeakpointsActivated) (CallFunction|ShouldToggleWeakpoints))
    (Class|BGALargeSpire|UpdateChildWeakpointActorbyKind (Variables|Internal|GetSpireActor) _targetstructdata_weakpointkind_14_3e0154dc496b6f9415a82583af261852 false)
    (Variables|Internal|SetbWeakpointsActivated false)
    (else
      _))
  (Utilities|Time|ClearandInvalidateTimerbyHandle (Variables|SpireSpike|GetSpawnSpireSpikeTimer))
  (Utilities|Array|Clear (Variables|SpireSpike|GetSpireSpikeSpawnLocation)))

(event Custom|SpawnSpirePike
  (bind _spawnedspirespikes (Variables|SpireSpike|GetSpawnedSpireSpikes))
  (bind _output (Utilities|Array|Get(acopy) _spawnedspirespikes))
  (if (not (Utilities|Array|IsEmpty (Variables|SpireSpike|GetSpireSpikeSpawnLocation)))
    (bind _returnvalue (Building|SpawnBuildingActor (Variables|SpireSpike|GetSpireSpikeClass) (Math|Transform|MakeTransform (Utilities|Array|Get(acopy) (Variables|SpireSpike|GetSpireSpikeSpawnLocation)) (Math|Rotator|MakeRotator 0.0 0.0 (Math|Random|RandomFloatinRange 0.0 3.6e+02)) (Math|Conversions|ToVector(Float) (ScalableFloat|GetValueatLevel (Variables|SpireSpike|GetSpireSpikeInitialScale))))))
    (Actor|SetOwner _returnvalue (Variables|Internal|GetSpireActor))
    (Utilities|FlowControl|Sequence
      (:then_0
        (Game|AssignOnDestroyed _returnvalue (delegate Custom|OnSpireSpikeDestroyed))
        (bind _returnvalue_2 (Utilities|Array|Add (Variables|SpireSpike|GetSpawnedSpireSpikes) (Utilities|Casting|CastToFortSpireSlag _returnvalue)))
        (if (> _returnvalue_2 (Variables|SpireSpike|GetMaxNumberOfSpireSpike))
          (Class|FortSpireSlag|SetLootDropsEnabled _output false)
          (Actor|DestroyActor _output)
          (Utilities|Array|RemoveIndex _spawnedspirespikes)))
      (:then_1
        (bind _asb_prop_spire_spike_shock_wave_growable (Utilities|Casting|CastToB_Prop_SpireSpike_ShockWave_Growable _returnvalue))
        (Utilities|IsValid (Actor|GetComponentbyClass _asb_prop_spire_spike_shock_wave_growable "/SpireSharedAssets/Components/Component_SpireSpike_SclarComponent.Component_SpireSpike_SclarComponent_C")
          (:Is%20Valid
            (Variables|/SpireSharedAssets/Components/Component_SpireSpike_SclarComponent.Component_SpireSpike_SclarComponent_C|SetSpikeMaxScaleArray (Utilities|Array|MakeArray (ScalableFloat|GetValueatLevel (Variables|SpireSpike|GetSF_SpikeMaxScale01)) (ScalableFloat|GetValueatLevel (Variables|SpireSpike|GetSF_SpikeMaxScale02)) (ScalableFloat|GetValueatLevel (Variables|SpireSpike|GetSF_SpikeMaxScale03)) (ScalableFloat|GetValueatLevel (Variables|SpireSpike|GetSF_SpikeMaxScale04))) (Actor|GetComponentbyClass _asb_prop_spire_spike_shock_wave_growable "/SpireSharedAssets/Components/Component_SpireSpike_SclarComponent.Component_SpireSpike_SclarComponent_C"))
            (Class|BPropSpireSpikeShockWaveGrowable|StartSpikeGrowth _asb_prop_spire_spike_shock_wave_growable))))
      (:then_2
        (Utilities|Array|RemoveIndex (Variables|SpireSpike|GetSpireSpikeSpawnLocation)))
      (:then_3
        (if (not (Utilities|Array|IsEmpty (Variables|SpireSpike|GetSpireSpikeSpawnLocation)))
          (bind _returnvalue_3 (Utilities|Time|SetTimerbyEvent (EventDispatchers|CreateEvent :!event Custom|SpawnSpirePike) (ScalableFloat|GetValueatLevel (Variables|SpireSpike|GetDelayBetweenSpireSpikePawn)) false true))
          (Variables|SpireSpike|SetSpawnSpireSpikeTimer _returnvalue_3))))))

(event Custom|OnSpireSpikeDestroyed
  (params (DestroyedActor :type Actor%20Object%20Reference))
  (bind _asfort_spire_slag (Utilities|Casting|CastToFortSpireSlag DestroyedActor))
  (Utilities|Array|RemoveItem (Variables|SpireSpike|GetSpawnedSpireSpikes) _asfort_spire_slag))

(event Ability|EventOnAbilityRemoved
  (for _array_element (Variables|SpireSpike|GetSpawnedSpireSpikes)
    (Utilities|IsValid _array_element
      (:Is%20Valid
        (Game|UnbindallEventsfromOnDestroyed _array_element))))
  (Utilities|Array|Clear (Variables|SpireSpike|GetSpawnedSpireSpikes)))

(fn ApplySlamDamage
  (params (TargetData :type Gameplay%20Ability%20Target%20Data%20Handle%20Structure))
  (Ability|ApplyGameplayEffectContainer TargetData "(TagName=\"Targeting.ApplicationTag.0\")"))

(fn GetCustomAbilitySourceTransform
  (return (Variables|Internal|GetImpactTransform)))

(fn GetValidSlamTarget
  (if (and (<= (Transformation|GetDistanceTo (Variables|Internal|GetSpireActor) (Variables|/SpireBoss/GameplayActors/BGA_Large_Spire.BGA_Large_Spire_C|GetCurrentTargetPlayerPawn (Variables|Internal|GetSpireActor))) (ScalableFloat|GetValueatLevel (Variables|Internal|GetSF_GroundWave_MaxTargetDist))) (not (GameplayTags|HasAnyMatchingGameplayTags (Variables|/SpireBoss/GameplayActors/BGA_Large_Spire.BGA_Large_Spire_C|GetCurrentTargetPlayerPawn (Variables|Internal|GetSpireActor)) (GameplayTags|MakeLiteralGameplayTagContainer "(GameplayTags=((TagName=\"Gameplay.Status.DBNO\"),(TagName=\"Athena.Gameplay.Status.Cloaked\")))"))))
    (return true (Variables|/SpireBoss/GameplayActors/BGA_Large_Spire.BGA_Large_Spire_C|GetCurrentTargetPlayerPawn (Variables|Internal|GetSpireActor)))
    (else
      (LocalVariables|SetbFoundPlayerTarget false)
      (bind _array_element (Utilities|Array|ForEachLoopwithBreak (Variables|/SpireBoss/GameplayActors/BGA_Large_Spire.BGA_Large_Spire_C|GetSensedPlayerPawns (Variables|Internal|GetSpireActor)))
        (:LoopBody
          (LocalVariables|SetTargetIter _array_element)
          (if (and (<= (Transformation|GetDistanceTo (Variables|Internal|GetSpireActor) _array_element) (ScalableFloat|GetValueatLevel (Variables|Internal|GetSF_GroundWave_MaxTargetDist))) (not (GameplayTags|HasAnyMatchingGameplayTags _array_element (GameplayTags|MakeLiteralGameplayTagContainer "(GameplayTags=((TagName=\"Gameplay.Status.DBNO\"),(TagName=\"Athena.Gameplay.Status.Cloaked\")))"))))
            (LocalVariables|SetbFoundPlayerTarget true)))
        (:Completed
          _))
      (if (LocalVariables|GetbFoundPlayerTarget)
        (return true (LocalVariables|GetTargetIter))
        (else
          (return false 0))))))

(fn InitSpireAbilityVariables
  (|CallParentFunction)
  (Variables|Internal|SetbWeakpointsActivated false)
  (Utilities|Array|Clear (Variables|Internal|GetInstantAOETargets)))

(fn CanActivateAbility
  (params (ActorInfo :type Gameplay%20Ability%20Actor%20Info%20Structure) (Handle :type Gameplay%20Ability%20Spec%20Handle%20Structure))
  (bind _asbga_large_spire (Utilities|Casting|CastToBGA_Large_Spire (Ability|GetAvatarActorfromActorInfo)))
  (LocalVariables|SetbHasTarget false)
  (bind _array_element (Utilities|Array|ForEachLoopwithBreak (Variables|/SpireBoss/GameplayActors/BGA_Large_Spire.BGA_Large_Spire_C|GetSensedPlayerPawns _asbga_large_spire))
    (:LoopBody
      (if (and (not (GameplayTags|HasAnyMatchingGameplayTags _array_element (GameplayTags|MakeLiteralGameplayTagContainer "(GameplayTags=((TagName=\"Gameplay.Status.DBNO\"),(TagName=\"Athena.Gameplay.Status.Cloaked\")))"))) (<= (Transformation|GetDistanceTo _asbga_large_spire _array_element) (ScalableFloat|GetValueatLevel (Variables|Internal|GetSF_GroundWave_MaxTargetDist))))
        (LocalVariables|SetbHasTarget true)))
    (:Completed
      _))
  (return 0 (Math|Boolean|ANDBoolean (not (GameplayTags|HasMatchingGameplayTag _asbga_large_spire "(TagName=\"Ability.SB.AppendageSlam.Cooldown\")")) (not (GameplayTags|HasMatchingGameplayTag _asbga_large_spire "(TagName=\"Ability.SB.InitialCombat\")")) (not (GameplayTags|HasMatchingGameplayTag _asbga_large_spire "(TagName=\"Ability.SB.Prevent\")")) (LocalVariables|GetbHasTarget))))

(fn SetupAbilityHitReact)

(fn SetupSpireAbility
  (|CallParentFunction)
  (bind _returnvalue (Ability|ApplyGameplayEffectToOwner (Variables|Internal|GetGESlamActive)))
  (Variables|Internal|SetSlamActiveHandle _returnvalue))

(fn ShouldInstantDamageAffectPawn
  (params (Player :type Fort%20Pawn%20Object%20Reference))
  (return (not (or (GameplayTags|HasAnyMatchingGameplayTags Player (Variables|Default|GetTagContainer_PreventKnockback)) (Team|OnSameTeam (Variables|Internal|GetSpireActor) Player)))))

(fn SpawnGroundWaveActor
  (bind _returnvalue (Building|SpawnBuildingGameplayActor (Variables|Internal|GetGroundWaveClass) (Math|Transform|MakeTransform (.location (Variables|Internal|GetImpactTransform))) (Variables|Internal|GetSpireActor) true false))
  (return _returnvalue))
