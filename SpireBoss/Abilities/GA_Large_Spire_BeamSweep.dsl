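; Gameplay ability for the Large Spire boss's beam sweep attack, gated by cooldown and state tags and a player sensed beyond a minimum distance.
; A montage exposes weakpoints, cues the targeted player and updates windup visuals until its Fire notify starts the boss's sweep.
; A looping timer then sphere-traces the beam, applying cached damage by hit type: vehicle, enemy player, player-built or other non-immune building, until the Outro.
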(event Ability|EventActivateAbility
  (|CallParentFunction)
  (CallFunction|CacheScalableFloatValues)
  (bind (_asynctaskproxy _notifytag) (Ability|Tasks|PlayMontageAndWaitNotify "SpireBoss_BeamSweep" "/SpireBoss/Assets/Animation/SpireBoss_Beam_Attack_M.SpireBoss_Beam_Attack_M" 1.0 "Intro")
    (:then
      (Variables|Internal|SetBeamSweepMontageTask _asynctaskproxy)
      (Class|BGALargeSpire|UpdateChildWeakpointActorbyKind (Variables|Internal|GetSpireActor) "NewEnumerator3" true)
      (Variables|Internal|SetbWeakpointsActive true)
      (GameplayCue|ExecuteGameplayCueOnActor(Burst) (Variables|/SpireBoss/GameplayActors/BGA_Large_Spire.BGA_Large_Spire_C|GetCurrentTargetPlayerPawn (Variables|Internal|GetSpireActor)) (Variables|Default|GetPlayer%20Targeted%20Tag) (Ability|GameplayCue|MakeGameplayCueParameters :NormalizedMagnitude 0.0 :RawMagnitude 4.0 :Location "0,0,0" :Normal "0,0,0" :Instigator (Variables|Internal|GetSpireActor) :SourceObject (Variables|/SpireBoss/GameplayActors/BGA_Large_Spire.BGA_Large_Spire_C|GetSkeletalMeshComponentBudgeted (Variables|Internal|GetSpireActor)) :GameplayEffectLevel 1 :AbilityLevel 1 :TargetAttachComponent (Variables|/Script/Engine.Character|GetMesh (Utilities|Casting|CastToFortPlayerPawnAthena (Variables|/SpireBoss/GameplayActors/BGA_Large_Spire.BGA_Large_Spire_C|GetCurrentTargetPlayerPawn (Variables|Internal|GetSpireActor)))) :bReplicateLocationWhenUsingMinimalRepProxy false)) ; reach out to FX if issue
      (bind _returnvalue (Utilities|Time|SetTimerbyEvent (delegate Custom|ProcessWindupVisUpdate) (ScalableFloat|GetValueatLevel (Variables|Internal|GetSF_WindupVisUpdateTime)) true))
      (Variables|Internal|SetHandle_WindupActive _returnvalue))
    (:OnCompleted
      _)
    (:OnInterrupted
      _)
    (:OnCancelled
      _)
    (:OnNotifyBegin
      (if (== _notifytag (Variables|Internal|GetMontageTag_Fire))
        (Utilities|Time|ClearandInvalidateTimerbyHandle (Variables|Internal|GetHandle_WindupActive))
        (CallFunction|StartBeam))))
  (Ability|EndAbility))

(event Custom|ProcessWindupVisUpdate
  (Class|BGALargeSpire|UpdatePreSweepTransform (Variables|Internal|GetSpireActor)))

(event Custom|StartBeam
  (Development|PrintString "@@@@@@ StartBeam [GA_Large_Spire_BeamSweep]: CALLED!" false (Variables|Debug|GetbDebugAbility))
  (Utilities|FlowControl|Sequence
    (:then_0
      (Class|BGALargeSpire|PerformBeamSweep (Variables|Internal|GetSpireActor)))
    (:then_1
      (bind _returnvalue (Utilities|Time|SetTimerbyEvent (EventDispatchers|CreateEvent :!event Custom|ProcessBeamHitUpdate) (CallFunction|GetBeamHitUpdateTime) true true))
      (Variables|HitProcessing|SetHandle_BeamHitTimer _returnvalue)
      (Variables|ForCreative|SetLastFrametime (Utilities|Time|GetGameTimeinSeconds))
      (Variables|ForCreative|SetPreviousBeamSweepYaw (ScalableFloat|GetValueatLevelwithContext (Variables|ForCreative|GetSF_BeamYawStart))))
    (:then_2
      (bind _returnvalue_2 (Ability|ApplyGameplayEffectToOwner (Variables|Default|GetActiveBeamGE)))
      (Variables|Internal|SetHandle_BeamActive _returnvalue_2))
    (:then_3
      (Utilities|FlowControl|Delay (Variables|HitProcessing|GetBeamFireTime))
      (Ability|Animation|MontageJumptoSection "Outro")
      (CallFunction|StopBeamHitUpdate))))

(event Custom|ProcessBeamHitUpdate
  (bind _outhits (CallFunction|HandleBeamTrace))
  (for _array_element _outhits
    (CallFunction|HandleBeamDamage _array_element)))

(event Custom|HandleBeamDamage
  (params (Hit :type Hit%20Result%20Structure%20%28by%20ref%29))
  (bind (_bblockinghit _binitialoverlap _time _distance _location _impactpoint _normal _impactnormal _physmat _hitactor) (Collision|BreakHitResult Hit))
  (bind _spireactor (Variables|Internal|GetSpireActor))
  (bind _bp_spiresanalyticscomponent (Variables|/SpireBoss/GameplayActors/BGA_Large_Spire.BGA_Large_Spire_C|GetBP_SpiresAnalyticsComponent _spireactor))
  (Variables|HitProcessing|SetAppliedFireDmg 0.0)
  (Variables|HitProcessing|SetCurrentTestActor _hitactor)
  (CallFunction|CalculateDamageforCurrentActor)
  (if (> (Variables|HitProcessing|GetAppliedFireDmg) 0)
    (Environment|ApplyGlobalEnvironmentDamagetoActor (Variables|HitProcessing|GetCurrentTestActor) (Variables|HitProcessing|GetAppliedFireDmg) (Variables|HitProcessing|GetBeamDamageContextTags))
    (Utilities|IsValid _bp_spiresanalyticscomponent
      (:Is%20Valid
        (Class|SpiresAnalyticsComponent|OnSpireDidDamage _bp_spiresanalyticscomponent (Variables|HitProcessing|GetCurrentTestActor) (Variables|HitProcessing|GetAppliedFireDmg) 0 (Variables|Internal|GetSpireActor) "0, 0, 0" 0 "None" "0, 0, 0")))))

(event Custom|StopBeamHitUpdate
  (Ability|RemoveGameplayEffectFromOwnerWithHandle (Variables|Internal|GetHandle_BeamActive))
  (Utilities|Time|ClearandInvalidateTimerbyHandle (Variables|HitProcessing|GetHandle_BeamHitTimer))
  (Utilities|Time|ClearandInvalidateTimerbyHandle (Variables|Internal|GetHandle_WindupActive))
  (Class|BGALargeSpire|UpdateChildWeakpointActorbyKind (Variables|Internal|GetSpireActor) "NewEnumerator3" false)
  (Variables|Internal|SetbWeakpointsActive false))

(event Ability|EventOnEndAbility
  (params (bWasCancelled :type Boolean))
  (|CallParentFunction bWasCancelled)
  (Ability|RemoveGameplayEffectFromOwnerWithHandle (Variables|Internal|GetBeamActiveHandle))
  (CallFunction|StopBeamHitUpdate)
  (if (Variables|Internal|GetbWeakpointsActive)
    (Class|BGALargeSpire|UpdateChildWeakpointActorbyKind (Variables|Internal|GetSpireActor) "NewEnumerator3" false)
    (Variables|Internal|SetbWeakpointsActive false)))

(fn CacheScalableFloatValues
  (Variables|HitProcessing|SetBeamFireTime (ScalableFloat|GetValueatLevelwithContextParam (Variables|Internal|GetSpireActor) (Variables|Internal|GetSF_BeamFireTime)))
  (Variables|HitProcessing|SetBeamVehicleDmg (ScalableFloat|GetValueatLevelwithContextParam (Variables|Internal|GetSpireActor) (Variables|HitProcessing|GetSF_BeamDmg_Vehicle)))
  (Variables|HitProcessing|SetBeamPlayerDmg (ScalableFloat|GetValueatLevelwithContextParam (Variables|Internal|GetSpireActor) (Variables|HitProcessing|GetSF_BeamDmg_Player)))
  (Variables|HitProcessing|SetBeamPlayerBuildingDmg (ScalableFloat|GetValueatLevelwithContextParam (Variables|Internal|GetSpireActor) (Variables|HitProcessing|GetSF_BeamDmg_BuildPlayer)))
  (Variables|HitProcessing|SetBeamBuildingDmg (ScalableFloat|GetValueatLevelwithContextParam (Variables|Internal|GetSpireActor) (Variables|HitProcessing|GetSF_BeamDmg_Building))))

(fn CalculateDamageForCurrentActor
  (LocalVariables|SetOwningActor (Ability|GetAvatarActorfromActorInfo)) ; cache just to slightly reduce clutter
  (Utilities|Casting|CastToFortVehicleInterface (Variables|HitProcessing|GetCurrentTestActor)
    (:then
      (Variables|HitProcessing|SetAppliedFireDmg (Variables|HitProcessing|GetBeamVehicleDmg)))
    (:CastFailed
      (bind _asfort_player_pawn (Utilities|Casting|CastToFortPlayerPawn (Variables|HitProcessing|GetCurrentTestActor))
        (:then
          (if (not (Team|OnSameTeam (Variables|Internal|GetSpireActor) _asfort_player_pawn))
            (Variables|HitProcessing|SetAppliedFireDmg (Variables|HitProcessing|GetBeamPlayerDmg))))
        (:CastFailed
          (bind _asbuilding_actor (Utilities|Casting|CastToBuildingActor (Variables|HitProcessing|GetCurrentTestActor)))
          (if (GameplayTags|HasMatchingGameplayTag _asbuilding_actor "(TagName=\"Gameplay.Status.Spire.Immune\")")
            _
            (elif (GameplayTags|HasMatchingGameplayTag _asbuilding_actor "(TagName=\"Building.Type.Player\")")
              (Variables|HitProcessing|SetAppliedFireDmg (Variables|HitProcessing|GetBeamPlayerBuildingDmg))
              (else
                (Variables|HitProcessing|SetAppliedFireDmg (Variables|HitProcessing|GetBeamBuildingDmg))))))))))

(fn GetBeamHitUpdateTime
  (return (ScalableFloat|GetValueatLevel (Variables|Internal|GetSF_BeamHitUpdateTime))))

(fn GetValidBeamTarget
  (if (>= (Transformation|GetDistanceTo (Variables|Internal|GetSpireActor) (Variables|/SpireBoss/GameplayActors/BGA_Large_Spire.BGA_Large_Spire_C|GetCurrentTargetPlayerPawn (Variables|Internal|GetSpireActor))) (ScalableFloat|GetValueatLevel (Variables|Internal|GetSF_BeamMinTargetDist)))
    (return true (Variables|/SpireBoss/GameplayActors/BGA_Large_Spire.BGA_Large_Spire_C|GetCurrentTargetPlayerPawn (Variables|Internal|GetSpireActor)))
    (else
      (LocalVariables|SetbFoundPlayerTarget false)
      (bind _array_element (Utilities|Array|ForEachLoopwithBreak (Variables|/SpireBoss/GameplayActors/BGA_Large_Spire.BGA_Large_Spire_C|GetSensedPlayerPawns (Variables|Internal|GetSpireActor)))
        (:LoopBody
          (LocalVariables|SetTargetIter _array_element)
          (if (>= (Transformation|GetDistanceTo (Variables|Internal|GetSpireActor) _array_element) (ScalableFloat|GetValueatLevel (Variables|Internal|GetSF_BeamMinTargetDist)))
            (LocalVariables|SetbFoundPlayerTarget true)))
        (:Completed
          _))
      (if (LocalVariables|GetbFoundPlayerTarget)
        (return true (LocalVariables|GetTargetIter))
        (else
          (return false 0))))))

(fn HandleBeamTrace
  (bind _transformbeamstoppos_location (Math|Transform|BreakTransform (Variables|/SpireBoss/GameplayActors/BGA_Large_Spire.BGA_Large_Spire_C|GetTransformBeamStopPos_Location (Variables|Internal|GetSpireActor))))
  (bind _currentbeamtransform_location (Math|Transform|BreakTransform (Variables|/SpireBoss/GameplayActors/BGA_Large_Spire.BGA_Large_Spire_C|GetCurrentBeamTransform_Location (Variables|Internal|GetSpireActor))))
  (bind _outhits (Collision|MultiSphereTraceForObjects _currentbeamtransform_location _transformbeamstoppos_location (ScalableFloat|GetValueatLevel (Variables|Internal|GetSF_BeamHitRadius)) (Variables|Internal|GetObject%20Types) false (Variables|Internal|GetAbilityActorsToIgnore) "None" true "(R=1.000000,G=0.000000,B=0.000000,A=1.000000)" "(R=0.000000,G=1.000000,B=0.000000,A=1.000000)" (Variables|HitProcessing|GetBeamFireTime)))
  (return _outhits))

(fn InitSpireAbilityVariables
  (bind (_bfoundtarget _beamtarget) (CallFunction|GetValidBeamTarget))
  (bind _spireactor (Variables|Internal|GetSpireActor))
  (|CallParentFunction)
  (if _bfoundtarget
    (bind _output_get (Variables|Internal|SetAbilityTargetPawn _beamtarget))
    (Variables|/SpireBoss/GameplayActors/BGA_Large_Spire.BGA_Large_Spire_C|SetBeamTargetPawn _output_get _spireactor)
    (Variables|Internal|SetAbilityActorsToIgnore (Utilities|Array|MakeArray (Variables|/Script/Engine.ChildActorComponent|GetChildActor (Variables|/SpireBoss/GameplayActors/BGA_Large_Spire.BGA_Large_Spire_C|GetWeakpoint01 _spireactor)) (Variables|/Script/Engine.ChildActorComponent|GetChildActor (Variables|/SpireBoss/GameplayActors/BGA_Large_Spire.BGA_Large_Spire_C|GetWeakpoint02 _spireactor)) (Variables|/Script/Engine.ChildActorComponent|GetChildActor (Variables|/SpireBoss/GameplayActors/BGA_Large_Spire.BGA_Large_Spire_C|GetWeakpoint03 _spireactor)) (Variables|/Script/Engine.ChildActorComponent|GetChildActor (Variables|/SpireBoss/GameplayActors/BGA_Large_Spire.BGA_Large_Spire_C|GetWeakpoint04 _spireactor)) _spireactor))
    (Variables|Internal|SetbWeakpointsActive false)
    (else
      (Ability|EndAbility))))

(fn CanActivateAbility
  (params (ActorInfo :type Gameplay%20Ability%20Actor%20Info%20Structure) (Handle :type Gameplay%20Ability%20Spec%20Handle%20Structure))
  (bind _asbga_large_spire (Utilities|Casting|CastToBGA_Large_Spire (Ability|GetAvatarActorfromActorInfo)))
  (LocalVariables|SetbHasTarget false)
  (bind _array_element (Utilities|Array|ForEachLoopwithBreak (Variables|/SpireBoss/GameplayActors/BGA_Large_Spire.BGA_Large_Spire_C|GetSensedPlayerPawns _asbga_large_spire))
    (:LoopBody
      (if (and (not (GameplayTags|HasAnyMatchingGameplayTags _array_element (GameplayTags|MakeLiteralGameplayTagContainer "(GameplayTags=((TagName=\"Gameplay.Status.DBNO\"),(TagName=\"Athena.Gameplay.Status.Cloaked\")))"))) (>= (Transformation|GetDistanceTo _asbga_large_spire _array_element) (ScalableFloat|GetValueatLevel (Variables|Internal|GetSF_BeamMinTargetDist))))
        (LocalVariables|SetbHasTarget true)))
    (:Completed
      _))
  (return 0 (Math|Boolean|ANDBoolean (not (GameplayTags|HasMatchingGameplayTag _asbga_large_spire "(TagName=\"Ability.SB.BeamSweep.Cooldown\")")) (not (GameplayTags|HasMatchingGameplayTag _asbga_large_spire "(TagName=\"Ability.SB.InitialCombat\")")) (not (GameplayTags|HasMatchingGameplayTag _asbga_large_spire "(TagName=\"Ability.SB.Prevent\")")) (LocalVariables|GetbHasTarget))))

(fn SetupSpireAbility
  (|CallParentFunction)
  (bind _returnvalue (Ability|ApplyGameplayEffectToOwner (Variables|Internal|GetGEBeamActive)))
  (Variables|Internal|SetBeamActiveHandle _returnvalue))
