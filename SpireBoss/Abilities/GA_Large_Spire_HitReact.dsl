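; Gameplay ability that plays the Large Spire boss's hit reaction and then flags a threshold-crossed summon.
; On activation it becomes non-cancelable and starts the spire actor's current hit-react montage through a stored task.
; The montage callbacks only print debug messages, and EndAbility does not wait on them.
; On end it applies the ThresholdCrossedSummon gameplay effect to its owner and stores the summon-ready handle.
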(event Ability|EventActivateAbility
  (bind _bdebugability (Variables|Debug|GetbDebugAbility))
  (Ability|SetCanbeCanceled false)
  (|CallParentFunction)
  (bind _asynctaskproxy (Ability|Tasks|PlayMontageAndWaitNotify "SpireBoss_HitReact" (Variables|/SpireBoss/GameplayActors/BGA_Large_Spire.BGA_Large_Spire_C|GetCurrentSpireHitReact (Variables|Internal|GetSpireActor)))
    (:then
      (Variables|Internal|SetHitReactMontageTask _asynctaskproxy))
    (:OnCompleted
      (Development|PrintString "@@@@@@ Event ActivateAbility [GA_Large_Spire_HitReact]: Montage Completed!" false _bdebugability))
    (:OnInterrupted
      (Development|PrintString "@@@@@@ Event ActivateAbility [GA_Large_Spire_HitReact]: Montage Interrupted!" false _bdebugability))
    (:OnCancelled
      (Development|PrintString "@@@@@@ Event ActivateAbility [GA_Large_Spire_HitReact]: Montage Cancelled!" false _bdebugability)))
  (Ability|EndAbility))

(event Ability|EventOnEndAbility
  (params (bWasCancelled :type Boolean))
  (|CallParentFunction bWasCancelled)
  (Development|PrintString "@@@@@@@@@ Event OnEndAbility [GA_Large_Spire_HitReact]: CALLED!" false (Variables|Debug|GetbDebugAbility))
  (bind _returnvalue (Ability|ApplyGameplayEffectToOwner (Variables|Internal|GetThresholdCrossedSummon)))
  (Variables|Internal|SetSummonReadyHandle _returnvalue))

(fn SetupAbilityHitReact)
