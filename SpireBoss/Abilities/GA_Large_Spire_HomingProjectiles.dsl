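; Spire boss gameplay ability that telegraphs and launches homing projectiles from sources placed around the spire.
; A rotated context actor runs three chained EQS queries, one per wedge, continuing past failures.
; Selected points, plus an optional bonus nearest the target, are shuffled and spawned on a timer during a montage, with optional weakpoints enabled.
; Once all spawn, they launch and the montage jumps to Outro; ability end destroys untriggered sources.
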(event Ability|EventActivateAbility
  (bind (_returnvalue_location _returnvalue_rotation) (Math|Transform|BreakTransform (Transformation|GetActorTransform (Variables|Internal|GetSpireActor))))
  (bind (_returnvalue_roll _returnvalue_pitch _returnvalue_yaw) (Math|Rotator|BreakRotator (Transformation|GetActorRotation (Variables|internal|GetActorLocationContext))))
  (|CallParentFunction)
  (bind _returnvalue (Building|SpawnBuildingGameplayActor (Variables|Default|GetActorLocationContextClass) (Math|Transform|MakeTransform _returnvalue_location _returnvalue_rotation) (Variables|Internal|GetSpireActor) true true))
  (bind _output_get (Variables|internal|SetActorLocationContext _returnvalue))
  ;[0.6094 0.0943 0 0.5] Process first source wedge.
  (bind _asbga_large_spire_proj_context (Utilities|Casting|CastToBGA_Large_Spire_Proj_Context _output_get)
    (:then
      (Variables|/SpireBoss/GameplayActors/BGA_Large_Spire_Proj_Context.BGA_Large_Spire_Proj_Context_C|SetSpireActor (Variables|Internal|GetSpireActor) _asbga_large_spire_proj_context)
      (Transformation|SetActorRotation (Variables|internal|GetActorLocationContext) (Math|Rotator|MakeRotator _returnvalue_roll _returnvalue_pitch (+ _returnvalue_yaw 9e+01)) false)
      (bind _returnvalue_2 (AI|EQS|RunEQSQuery (Variables|Default|GetEQS_ProjectileSource) (Variables|internal|GetActorLocationContext) "AllMatching"))
      (Utilities|IsValid _returnvalue_2
        (:Is%20Valid
          (AI|EQS|BindEventtoOnQueryFinished _returnvalue_2 (delegate Custom|OnFirstAngleQueryFinished)))
        (:Is%20Not%20Valid
          (Development|PrintString "@@@@@@ Event ActivateAbility [GA_Large_Spire_HomingProjectiles]: First query ERROR - Player Query Instance INVALID!" false (Variables|Debug|GetbDebugAbility))
          (CallFunction|ProcessSecondAngleQuery))))
    (:CastFailed
      (Actor|DestroyActor (Variables|internal|GetActorLocationContext))
      (Ability|EndAbility))))

(event Custom|OnFirstAngleQueryFinished
  (params (QueryInstance :type EQS%20Query%20Instance%20Object%20Reference) (QueryStatus :type EEnvQueryStatus%20Enum :default "Processing"))
  (bind _returnvalue (CallFunction|ShouldUseBonusLocation))
  ;[0.6094 0.0943 0 0.5] Process first source wedge.
  (if (Utilities|Enum|Equal(Enum) QueryStatus "Success")
    (bind (_resultlocations _returnvalue_2) (AI|EQS|GetQueryResultsasLocations QueryInstance))
    (if _returnvalue_2
      (Default|DebugPrintArray 0 _resultlocations "(R=1.000000,G=0.000000,B=0.949739,A=1.000000)")
      (bind (_selectedsourcelocs _bonusloc) (CallFunction|SelectProjectileSourceLocs _resultlocations _returnvalue))
      (if _returnvalue
        (CallFunction|ProcessBonusLoc _bonusloc))
      (Utilities|Array|AppendArray (Variables|Internal|GetProjectileSourceArray) _selectedsourcelocs)
      (CallFunction|ProcessSecondAngleQuery)
      (else
        (Development|PrintString "@@@@@@ Event ActivateAbility [GA_Large_Spire_HomingProjectiles]: First query ERROR - NO valid reults for FIRST scan!" false (Variables|Debug|GetbDebugAbility))
        (CallFunction|ProcessSecondAngleQuery)))
    (else
      (Development|PrintString "@@@@@@ Event ActivateAbility [GA_Large_Spire_HomingProjectiles]: First query ERROR -Player Query Status NOT Success!" false (Variables|Debug|GetbDebugAbility))
      (CallFunction|ProcessSecondAngleQuery))))

(event Custom|ProcessSecondAngleQuery
  (bind (_returnvalue_roll _returnvalue_pitch _returnvalue_yaw) (Math|Rotator|BreakRotator (Transformation|GetActorRotation (Variables|internal|GetActorLocationContext))))
  ;[0.003455 0.375 0.1965] Process second source wedge.
  (Transformation|SetActorRotation (Variables|internal|GetActorLocationContext) (Math|Rotator|MakeRotator _returnvalue_roll _returnvalue_pitch (+ _returnvalue_yaw 1.2e+02)) false)
  (bind _returnvalue (AI|EQS|RunEQSQuery (Variables|Default|GetEQS_ProjectileSource) (Variables|internal|GetActorLocationContext) "AllMatching"))
  (Utilities|IsValid _returnvalue
    (:Is%20Valid
      (AI|EQS|BindEventtoOnQueryFinished _returnvalue (delegate Custom|OnSecondAngleQueryFinished)))
    (:Is%20Not%20Valid
      (Development|PrintString "@@@@@@ Event ActivateAbility [GA_Large_Spire_HomingProjectiles]: Second query ERROR - Player Query Instance INVALID!" false (Variables|Debug|GetbDebugAbility))
      (CallFunction|ProcessThirdAngleQuery))))

(event Custom|OnSecondAngleQueryFinished
  (params (QueryInstance :type EQS%20Query%20Instance%20Object%20Reference) (QueryStatus :type EEnvQueryStatus%20Enum :default "Processing"))
  (bind _returnvalue (CallFunction|ShouldUseBonusLocation))
  ;[0.003455 0.375 0.1965] Process second source wedge.
  (if (Utilities|Enum|Equal(Enum) QueryStatus "Success")
    (bind (_resultlocations _returnvalue_2) (AI|EQS|GetQueryResultsasLocations QueryInstance))
    (if _returnvalue_2
      (Default|DebugPrintArray 0 _resultlocations "(R=0.000000,G=1.000000,B=0.636741,A=1.000000)")
      (bind (_selectedsourcelocs _bonusloc) (CallFunction|SelectProjectileSourceLocs _resultlocations _returnvalue))
      (if _returnvalue
        (CallFunction|ProcessBonusLoc _bonusloc))
      (Utilities|Array|AppendArray (Variables|Internal|GetProjectileSourceArray) _selectedsourcelocs)
      (CallFunction|ProcessThirdAngleQuery)
      (else
        (Development|PrintString "@@@@@@ Event ActivateAbility [GA_Large_Spire_HomingProjectiles]: Second query ERROR - NO valid reults for SECOND scan!" false (Variables|Debug|GetbDebugAbility))
        (CallFunction|ProcessThirdAngleQuery)))
    (else
      (Development|PrintString "@@@@@@ Event ActivateAbility [GA_Large_Spire_HomingProjectiles]: Second query ERROR -Player Query Status NOT Success!" false (Variables|Debug|GetbDebugAbility))
      (CallFunction|ProcessThirdAngleQuery))))

(event Custom|ProcessThirdAngleQuery
  (bind (_returnvalue_roll _returnvalue_pitch _returnvalue_yaw) (Math|Rotator|BreakRotator (Transformation|GetActorRotation (Variables|internal|GetActorLocationContext))))
  ;[0.2633 0.02256 0.9387] Process third source wedge.
  (Transformation|SetActorRotation (Variables|internal|GetActorLocationContext) (Math|Rotator|MakeRotator _returnvalue_roll _returnvalue_pitch (+ _returnvalue_yaw 1.2e+02)) false)
  (bind _returnvalue (AI|EQS|RunEQSQuery (Variables|Default|GetEQS_ProjectileSource) (Variables|internal|GetActorLocationContext) "AllMatching"))
  (Utilities|IsValid _returnvalue
    (:Is%20Valid
      (AI|EQS|BindEventtoOnQueryFinished _returnvalue (delegate Custom|OnThirdAngleQueryFinished)))
    (:Is%20Not%20Valid
      (Development|PrintString "@@@@@@ Event ActivateAbility [GA_Large_Spire_HomingProjectiles]: Third query ERROR - Player Query Instance INVALID!" false)
      (CallFunction|PlayProjectileMontage))))

(event Custom|OnThirdAngleQueryFinished
  (params (QueryInstance :type EQS%20Query%20Instance%20Object%20Reference) (QueryStatus :type EEnvQueryStatus%20Enum :default "Processing"))
  (bind _returnvalue (CallFunction|ShouldUseBonusLocation))
  ;[0.2633 0.02256 0.9387] Process third source wedge.
  (if (Utilities|Enum|Equal(Enum) QueryStatus "Success")
    (bind (_resultlocations _returnvalue_2) (AI|EQS|GetQueryResultsasLocations QueryInstance))
    (if _returnvalue_2
      (Default|DebugPrintArray 0 _resultlocations "(R=1.000000,G=0.254537,B=0.030821,A=1.000000)")
      (bind (_selectedsourcelocs _bonusloc) (CallFunction|SelectProjectileSourceLocs _resultlocations _returnvalue))
      (if _returnvalue
        (CallFunction|ProcessBonusLoc _bonusloc))
      (Utilities|Array|AppendArray (Variables|Internal|GetProjectileSourceArray) _selectedsourcelocs)
      (CallFunction|PlayProjectileMontage)
      (else
        (Development|PrintString "@@@@@@ Event ActivateAbility [GA_Large_Spire_HomingProjectiles]: Third query ERROR - NO valid reults for SECOND scan!" false (Variables|Debug|GetbDebugAbility))
        (CallFunction|PlayProjectileMontage)))
    (else
      (Development|PrintString "@@@@@@ Event ActivateAbility [GA_Large_Spire_HomingProjectiles]: Third query ERROR -Player Query Status NOT Success!" false (Variables|Debug|GetbDebugAbility))
      (CallFunction|PlayProjectileMontage))))

(event Custom|PlayProjectileMontage
  (bind _asynctaskproxy (Ability|Tasks|PlayMontageAndWaitNotify "SpireBoss_HomingProjectiles" (Variables|Internal|GetHomingProjectileMontage) 1.0 "Intro")
    (:then
      (Variables|Internal|SetHomingProjectileMontageTask _asynctaskproxy)
      (if (CallFunction|ShouldToggleWeakpoints)
        (CallFunction|UpdateHomingProjectileWeakpoints true)
        (Variables|Internal|SetbWeakpointsActivated true)
        (else
          _))
      (CallFunction|SpawnProjectileSources))
    (:OnCompleted
      _)
    (:OnInterrupted
      _)
    (:OnCancelled
      _))
  (CallFunction|OnProjectileMontageEnd))

(event Custom|SpawnProjectileSources
  (bind _sf_homingprojsourcesperarc (Variables|Internal|GetSF_HomingProjSourcesPerArc))
  (bind _returnvalue (ScalableFloat|GetValueatLevel _sf_homingprojsourcesperarc))
  (bind _returnvalue_2 (* _returnvalue 3.0))
  (Development|PrintString "@@@@@@ SpawnProjectileSources [GA_Large_Spire_HomingProjectiles]: CALLED!" false (Variables|Debug|GetbDebugAbility))
  (CallFunction|PrepareProjSourceArrayforSpawn)
  (Variables|Internal|SetMaxProjectileSources (Math|Float|Truncate (Math|Float|SelectFloat (+ _returnvalue_2 1.0) _returnvalue_2 (CallFunction|ShouldUseBonusLocation))))
  (Variables|Internal|SetProjectileSourceCounter 0)
  (bind _returnvalue_3 (Utilities|Time|SetTimerbyEvent (delegate Custom|CreateNextSource) (/ (+ (Variables|Internal|GetProjectileMontageIntroTime) (ScalableFloat|GetValueatLevel (Variables|Internal|GetSF_HomingProjRecoveryTime))) _returnvalue_2) true true))
  (Variables|Internal|SetTimerHandle_ProjectileSourceCreation _returnvalue_3))

(event Custom|CreateNextSource
  (bind _projectilesourcecounter (Variables|Internal|GetProjectileSourceCounter))
  (bind _projectilesourcearray (Variables|Internal|GetProjectileSourceArray))
  (bind _output (Utilities|Array|Get(acopy) _projectilesourcearray _projectilesourcecounter))
  (Variables|Internal|SetSourceLocIter _output)
  (CallFunction|SpawnProjSource (Utilities|Operators|Add _output (Math|Vector|MakeVector 0.0 0.0 (Math|Random|RandomFloatinRange (ScalableFloat|GetValueatLevel (Variables|Internal|GetSF_HomingProjSourceDownMax)) (ScalableFloat|GetValueatLevel (Variables|Internal|GetSF_HomingProjSourceUpMax))))))
  (Math|Integer|IncrementInt (Variables|Internal|GetProjectileSourceCounter)
    (:%20%20
      (if (CallFunction|IsFinishedSpawningProjSources)
        (Utilities|Time|ClearandInvalidateTimerbyHandle (Variables|Internal|GetTimerHandle_ProjectileSourceCreation))
        (CallFunction|TriggerProjectileLaunch)))))

(event Custom|TriggerProjectileLaunch
  (Development|PrintString "@@@@@@ TriggerProjectileLaunch [GA_Large_Spire_HomingProjectiles]: CALLED!" false (Variables|Debug|GetbDebugAbility))
  (if (CallFunction|ShouldToggleWeakpoints)
    (CallFunction|UpdateHomingProjectileWeakpoints false)
    (Variables|Internal|SetbWeakpointsActivated false)
    (else
      _))
  (for _array_element (Variables|Internal|GetHomingProjectileSourceArray)
    (Class|BGALargeSpireProjSource|InitiateProjectileLaunch _array_element self))
  (Ability|Animation|MontageJumptoSection "Outro"))

(event Ability|EventOnEndAbility
  (params (bWasCancelled :type Boolean))
  (|CallParentFunction bWasCancelled)
  (Development|PrintString "@@@@@@@@@ Event OnEndAbility [GA_Large_Spire_HomingProjectiles]: CALLED!" false (Variables|Debug|GetbDebugAbility))
  (Ability|RemoveGameplayEffectFromOwnerWithHandle (Variables|Internal|GetProjectilesActiveHandle))
  (Actor|DestroyActor (Variables|internal|GetActorLocationContext))
  (if (and (CallFunction|ShouldToggleWeakpoints) (Variables|Internal|GetbWeakpointsActivated))
    (CallFunction|UpdateHomingProjectileWeakpoints false)
    (Variables|Internal|SetbWeakpointsActivated false))
  (for _array_element (Variables|Internal|GetHomingProjectileSourceArray)
    (Utilities|IsValid _array_element
      (:Is%20Valid
        (if (not (Variables|/SpireBoss/GameplayActors/BGA_Large_Spire_Proj_Source.BGA_Large_Spire_Proj_Source_C|GetbIsSourceTriggered _array_element))
          (Actor|DestroyActor _array_element))))))

(fn DebugPrintArray
  (params (LocationArray :type Array%20of%20Vectors) (LocationColor :type Linear%20Color%20Structure))
  (bind _bdebugability (Variables|Debug|GetbDebugAbility))
  (Development|PrintString (Utilities|String|Append "@@@@@@@@@ DebugPrintArray [GA_Large_Spire_HomingProjectiles]: CALLED with a Location Array of size " (Utilities|String|ToString(Integer) (Utilities|Array|Length (Variables|GetLocationArray))) "!") false _bdebugability)
  (if _bdebugability
    (for _array_element (Variables|GetLocationArray)
      (Rendering|Debug|DrawDebugSphere _array_element 5e+01 12 (Variables|GetLocationColor) 1e+01 2.0))))

(fn InitSpireAbilityVariables
  (|CallParentFunction)
  (Utilities|Array|Clear (Variables|Internal|GetHomingProjectileSourceArray))
  (Utilities|Array|Clear (Variables|Internal|GetProjectileSourceArray))
  (Variables|Internal|SetCurrentBestBonusLoc "0, 0, 0")
  (Variables|Internal|SetCurrentBestBonusDist 1e+17)
  (Variables|Internal|SetAbilityTarget (Variables|/SpireBoss/GameplayActors/BGA_Large_Spire.BGA_Large_Spire_C|GetCurrentTargetPlayerPawn (Variables|Internal|GetSpireActor)))
  (Variables|Internal|SetbWeakpointsActivated false))

(fn IsFinishedSpawningProjSources
  (return (>= (Variables|Internal|GetProjectileSourceCounter) (Variables|Internal|GetMaxProjectileSources))))

(fn CanActivateAbility
  (params (ActorInfo :type Gameplay%20Ability%20Actor%20Info%20Structure) (Handle :type Gameplay%20Ability%20Spec%20Handle%20Structure))
  (bind _asbga_large_spire (Utilities|Casting|CastToBGA_Large_Spire (Ability|GetAvatarActorfromActorInfo)))
  (return 0 (Math|Boolean|ANDBoolean (not (GameplayTags|HasMatchingGameplayTag _asbga_large_spire "(TagName=\"Ability.SB.HomingProjectiles.Cooldown\")")) (not (GameplayTags|HasMatchingGameplayTag _asbga_large_spire "(TagName=\"Ability.SB.InitialCombat\")")) (not (GameplayTags|HasMatchingGameplayTag _asbga_large_spire "(TagName=\"Ability.SB.Prevent\")")) (Class|BGALargeSpire|HasValidAbilityTarget _asbga_large_spire))))

(fn OnProjectileMontageEnd
  (Ability|EndAbility))

(fn PrepareProjSourceArrayForSpawn
  (if (CallFunction|ShouldUseBonusLocation)
    (Utilities|Array|Add (Variables|Internal|GetProjectileSourceArray) (Variables|Internal|GetCurrentBestBonusLoc))
    (if (Variables|Debug|GetbDebugAbility)
      (Rendering|Debug|DrawDebugSphere (Variables|Internal|GetCurrentBestBonusLoc) 1e+02 12 "(R=0.942708,G=0.018629,B=0.014127,A=1.000000)" 1e+01 2.0))
    (else
      _))
  (Utilities|Array|Shuffle (Variables|Internal|GetProjectileSourceArray)))

(fn ProcessBonusLoc
  (params (BonusLoc :type Vector))
  (bind _abilitytarget (Variables|Internal|GetAbilityTarget))
  (bind _returnvalue (Pawn|GetNavAgentLocation _abilitytarget))
  (bind _returnvalue_2 (Math|Vector|Distance(Vector) BonusLoc _returnvalue))
  (LocalVariables|SetTestDistance _returnvalue_2)
  (if (< _returnvalue_2 (Variables|Internal|GetCurrentBestBonusDist))
    (Variables|Internal|SetCurrentBestBonusLoc (Variables|GetBonusLoc))
    (Variables|Internal|SetCurrentBestBonusDist (LocalVariables|GetTestDistance))))

(fn SelectProjectileSourceLocs
  (params (LocationArray :type Array%20of%20Vectors) (ReturnBonusLoc :type Boolean))
  (bind _sourcelocarray (LocalVariables|GetSourceLocArray))
  (bind _sentlocationarray (LocalVariables|GetSentLocationArray))
  (bind _output_get (LocalVariables|SetSentLocationArray LocationArray))
  (Utilities|Array|Clear _sourcelocarray)
  (Utilities|Array|AddUnique _sourcelocarray (Utilities|Array|Get(acopy) _output_get))
  (Utilities|Array|RemoveIndex _sentlocationarray)
  (Utilities|Array|Shuffle _sentlocationarray)
  (LocalVariables|SetSourceLocCounter 1.0)
  (while (< (LocalVariables|GetSourceLocCounter) (ScalableFloat|GetValueatLevel (Variables|Internal|GetSF_HomingProjSourcesPerArc)))
    (Utilities|Array|AddUnique (LocalVariables|GetSourceLocArray) (Utilities|Array|Get(acopy) (LocalVariables|GetSentLocationArray) (Math|Float|Truncate (LocalVariables|GetSourceLocCounter))))
    (Math|Float|IncrementFloat (LocalVariables|GetSourceLocCounter)))
  (if (Variables|Debug|GetbDebugAbility)
    (for _array_element (LocalVariables|GetSourceLocArray)
      (Rendering|Debug|DrawDebugSphere _array_element 1e+02 12 "(R=0.942708,G=0.018629,B=0.014127,A=1.000000)" 1e+01 2.0))
    (else
      _))
  (if (Variables|GetReturnBonusLoc)
    (return (LocalVariables|GetSourceLocArray) (Utilities|Array|Get(acopy) (LocalVariables|GetSentLocationArray) (Math|Float|Truncate (LocalVariables|GetSourceLocCounter))))
    (else
      (return (LocalVariables|GetSourceLocArray) "0, 0, 0"))))

(fn SetupSpireAbility
  (|CallParentFunction)
  (bind _returnvalue (Ability|ApplyGameplayEffectToOwner (Variables|Internal|GetGEProjectilesActive)))
  (Variables|Internal|SetProjectilesActiveHandle _returnvalue))

(fn ShouldUseBonusLocation
  (return (ScalableFloat|GetValueasBool (Variables|Internal|GetSF_HomingProjGetBonusLoc))))

(fn SpawnProjSource
  (params (Transform_Location :type Vector))
  (bind (_returnvalue_x _returnvalue_y) (Math|Vector|BreakVector (Transformation|GetActorLocation (Variables|Internal|GetSpireActor))))
  (bind (_sourcelociter_x _sourcelociter_y) (Math|Vector|BreakVector (Variables|GetSourceLocIter_X)))
  (bind _returnvalue (Building|SpawnBuildingGameplayActor (Variables|Internal|GetProjectileSourceClass) (Math|Transform|MakeTransform Transform_Location (Math|Rotator|MakeRotfromY (Utilities|Operators|Subtract (Math|Vector|MakeVector _sourcelociter_x _sourcelociter_y) (Math|Vector|MakeVector _returnvalue_x _returnvalue_y)))) (Variables|Internal|GetSpireActor) true false))
  (bind _asbga_large_spire_proj_source (Utilities|Casting|CastToBGA_Large_Spire_Proj_Source _returnvalue))
  (Utilities|Array|AddUnique (Variables|Internal|GetHomingProjectileSourceArray) _asbga_large_spire_proj_source)
  (bind _returnvalue_location (Math|Transform|BreakTransform (Transformation|GetActorTransform _asbga_large_spire_proj_source)))
  (GameplayCue|ExecuteGameplayCueOnActor(Burst) (Variables|Internal|GetSpireActor) (Variables|Ability|GetGameplay%20Cue%20Projectile%20Telegraph) (Ability|GameplayCue|MakeGameplayCueParameters :NormalizedMagnitude 0.0 :RawMagnitude 0.0 :Location _returnvalue_location :Normal "0,0,0" :Instigator (Variables|Internal|GetSpireActor) :GameplayEffectLevel 1 :AbilityLevel 1 :bReplicateLocationWhenUsingMinimalRepProxy false))
  (Class|BGALargeSpireProjSource|SetOwningSpire _asbga_large_spire_proj_source (Variables|Internal|GetSpireActor))
  (return _asbga_large_spire_proj_source))

(fn UpdateHomingProjectileWeakpoints
  (params (Enabled :type Boolean))
  (for _array_element (LocalVariables|GetHomingWeakpointKinds)
    (Class|BGALargeSpire|UpdateChildWeakpointActorbyKind (Variables|Internal|GetSpireActor) _array_element (Variables|GetEnabled)))
  (return))
