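; Gameplay ability for the Large Spire boss that reveals it and plays its intro animation.
; On activation it unhides the avatar actor, then starts a PlayMontageAndWait task with the Intro Montage variable.
; EndAbility follows directly on the task's immediate output rather than waiting for completion.
; The completed, interrupted and cancelled callbacks and the OnEndAbility event are empty.
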(event Ability|EventActivateAbility
  (Rendering|SetActorHiddenInGame (Ability|GetAvatarActorfromActorInfo) false)
  (Ability|Tasks|PlayMontageAndWait "None" (Variables|Default|GetIntro%20Montage)
    (:OnCompleted
      _)
    (:OnInterrupted
      _)
    (:OnCancelled
      _))
  (Ability|EndAbility))

(event Ability|EventOnEndAbility
  (params (bWasCancelled :type Boolean)))
