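; Large Spire boss gameplay ability that summons by starting spawn anticipations through the encounter manager.
; Activation requires prefabs awaiting anticipation and the Ready tag, with no Cooldown or InitialCombat tag.
; It applies a summoning effect, removes the Ready-granting effect and plays the summon montage.
; The montage's Fire notify triggers server-only anticipation start, while other notifies fire a roar cue at the shard center socket.
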(event Ability|EventActivateAbility
  (bind _returnvalue_location (Math|Transform|BreakTransform (Transformation|GetSocketTransform (Variables|/SpireBoss/GameplayActors/BGA_Large_Spire.BGA_Large_Spire_C|GetSkeletalMeshComponentBudgeted (Variables|Internal|GetSpireActor)) "shard_center_mask" "RTS_World")))
  (bind (_montageinfo_montagetoplay _montageinfo_animplayrate _montageinfo_animrootmotiontranslationscale) (Utilities|Struct|BreakFortGameplayAbilityMontageInfo (Variables|GetMontageInfo_MontageToPlay)))
  (|CallParentFunction)
  (bind _returnvalue (Ability|ApplyGameplayEffectToOwner (Variables|Internal|GetGE_AbilitySummoning)))
  (Variables|Internal|SetHdl_AbilitySummoning _returnvalue)
  (Ability|RemoveGameplayEffectFromOwnerWithGrantedTags "(GameplayTags=((TagName=\"Ability.SB.Summon.Ready\")))")
  (bind (_asynctaskproxy _notifytag) (Ability|Tasks|PlayMontageAndWaitNotify "SpireBoss_Summon" _montageinfo_montagetoplay _montageinfo_animplayrate "None" true _montageinfo_animrootmotiontranslationscale)
    (:then
      (Variables|Internal|SetSummonMontageTask _asynctaskproxy))
    (:OnCompleted
      _)
    (:OnInterrupted
      _)
    (:OnCancelled
      _)
    (:OnNotifyBegin
      (if (== _notifytag (Variables|Internal|GetMontageTag_Fire))
        (CallFunction|TriggerSpawnAnticipators)
        (GameplayCue|ExecuteGameplayCueOnActor(Burst) (Variables|Internal|GetSpireActor) (Variables|Ability|GetGC%20Roar) (Ability|GameplayCue|MakeGameplayCueParameters :NormalizedMagnitude 1.0 :RawMagnitude 0.0 :Location _returnvalue_location :Normal "0,0,0" :GameplayEffectLevel 1 :AbilityLevel 1 :TargetAttachComponent (Variables|/SpireBoss/GameplayActors/BGA_Large_Spire.BGA_Large_Spire_C|GetSkeletalMeshComponentBudgeted (Variables|Internal|GetSpireActor)) :bReplicateLocationWhenUsingMinimalRepProxy false)))))
  (Ability|EndAbility))

(event Custom|TriggerSpawnAnticipators
  (bind _spireactor (Variables|Internal|GetSpireActor))
  (bind _encountermanager (Variables|/SpireBoss/GameplayActors/BGA_Large_Spire.BGA_Large_Spire_C|GetEncounterManager _spireactor))
  (Development|PrintString "@@@@@@ TriggerSpawnAnticipators [GA_Large_Spire_Summon]: CALLED!" false (Variables|Debug|GetbDebugAbility))
  (if (Networking|IsServer)
    (Utilities|IsValid _spireactor
      (:Is%20Valid
        (Utilities|IsValid _encountermanager
          (:Is%20Valid
            (Class|BPSpireEncounterManager|StartAnticipations _encountermanager)))))))

(fn CanActivateAbility
  (params (ActorInfo :type Gameplay%20Ability%20Actor%20Info%20Structure) (Handle :type Gameplay%20Ability%20Spec%20Handle%20Structure))
  (bind _asbga_large_spire (Utilities|Casting|CastToBGA_Large_Spire (Ability|GetAvatarActorfromActorInfo))
    (:then
      (Utilities|IsValid (Variables|/SpireBoss/GameplayActors/BGA_Large_Spire.BGA_Large_Spire_C|GetEncounterManager _asbga_large_spire)
        (:Is%20Valid
          (return 0 (Math|Boolean|ANDBoolean (> (Utilities|Array|Length (Variables|/SpireSharedAssets/Lager/BP_Spire_Encounter_Manager.BP_Spire_Encounter_Manager_C|GetPrefabsWaitingAnticipation (Variables|/SpireBoss/GameplayActors/BGA_Large_Spire.BGA_Large_Spire_C|GetEncounterManager _asbga_large_spire))) 0) (not (GameplayTags|HasMatchingGameplayTag _asbga_large_spire "(TagName=\"Ability.SB.Summon.Cooldown\")")) (not (GameplayTags|HasMatchingGameplayTag _asbga_large_spire "(TagName=\"Ability.SB.InitialCombat\")")) (GameplayTags|HasMatchingGameplayTag _asbga_large_spire "(TagName=\"Ability.SB.Summon.Ready\")"))))
        (:Is%20Not%20Valid
          _)))
    (:CastFailed
      _))
  (return 0 false))
